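; Looping gameplay cue for the Rocket Ram item's braking state, which handles a slide sound on the affected player.
; On application it runs the parent logic, then spawns the slide sound attached to the player pawn's root component if no stored slide audio component is valid, and caches it.
; On removal it runs the parent logic and fades that component out without clearing the stored reference.
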(event EventOnApplication(EitherVFX-preferthisone)
  (params (MyTarget :type Actor%20Object%20Reference) (Parameters :type Gameplay%20Cue%20Parameters%20Structure%20%28by%20ref%29) (ParticleComponents :type Array%20of%20FXSystem%20Component%20Object%20References) (AudioComponents :type Array%20of%20Audio%20Component%20Object%20References) (BurstCameraShakeInstance :type Camera%20Shake%20Base%20Object%20Reference) (BurstDecalInstance :type Decal%20Actor%20Object%20Reference))
  (|CallParentFunction MyTarget Parameters ParticleComponents AudioComponents BurstCameraShakeInstance BurstDecalInstance)
  (bind _asplayer_pawn_athena (Utilities|Casting|CastToPlayerPawn_Athena MyTarget))
  (Variables|Default|GetSlideAudioComponent
    (:then)
    (:else
      (bind _returnvalue (Audio|SpawnSoundAttached (Variables|Default|GetSlideSoundCue) (Variables|/Script/Engine.Actor|GetRootComponent _asplayer_pawn_athena)))
      (Variables|Default|SetSlideAudioComponent _returnvalue))))

(event EventOnRemoval(EitherVFX-preferthisversion)
  (params (MyTarget :type Actor%20Object%20Reference) (Parameters :type Gameplay%20Cue%20Parameters%20Structure%20%28by%20ref%29) (ParticleComponents :type Array%20of%20FXSystem%20Component%20Object%20References) (AudioComponents :type Array%20of%20Audio%20Component%20Object%20References) (BurstCameraShakeInstance :type Camera%20Shake%20Base%20Object%20Reference) (BurstDecalInstance :type Decal%20Actor%20Object%20Reference))
  (|CallParentFunction MyTarget Parameters ParticleComponents AudioComponents BurstCameraShakeInstance BurstDecalInstance)
  (bind _slideaudiocomponent (Variables|Default|GetSlideAudioComponent)
    (:then
      (Audio|Components|Audio|FadeOut _slideaudiocomponent 0.5 0.0 "Linear" :!class /Script/Engine.AudioComponent))
    (:else)))

(fn ConstructionScript)
