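; Looping gameplay cue that provides audio feedback for the Rocket Ram cooldown.
; On start it caches the target pawn and, only if locally viewed, runs a repeating one-second timer that plays a 2D tick sound while the Rocket Ram is equipped.
; On removal it clears the timer and, if the Rocket Ram is equipped, spawns a stop sound attached to the weapon, without a local-view check.
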(event EventBeginPlay)

(event Collision|EventActorBeginOverlap
  (params (OtherActor :type Actor%20Object%20Reference)))

(event EventTick
  (params (DeltaSeconds :type Float%20%28single-precision%29)))

(event EventOnLoopingStart(EitherVFX-preferthisversion)
  (params (MyTarget :type Actor%20Object%20Reference) (Parameters :type Gameplay%20Cue%20Parameters%20Structure%20%28by%20ref%29) (ParticleComponents :type Array%20of%20FXSystem%20Component%20Object%20References) (AudioComponents :type Array%20of%20Audio%20Component%20Object%20References))
  (|CallParentFunction MyTarget Parameters ParticleComponents AudioComponents)
  (bind _asfort_player_pawn_athena (Utilities|Casting|CastToFortPlayerPawnAthena MyTarget))
  (Utilities|IsValid _asfort_player_pawn_athena
    (:Is%20Valid
      (bind _output_get (Variables|Default|SetPlayerRef _asfort_player_pawn_athena))
      (if (Pawn|IsLocallyViewed _output_get)
        (bind _returnvalue (Utilities|Time|SetTimerbyEvent (delegate Custom|PlayCooldownSound) 1.0 true))
        (Variables|Default|SetTickSoundTimerHandle _returnvalue)))))

(event Custom|PlayCooldownSound
  (bind _playerref (Variables|Default|GetPlayerRef))
  (bind _returnvalue (Weapon|GetCurrentWeapon _playerref))
  (Utilities|IsValid _returnvalue
    (:Is%20Valid
      (bind _asb_weapon_rocket_ram (Utilities|Casting|CastToB_Weapon_RocketRam _returnvalue))
      (if (FortWeapon|IsEquipped _asb_weapon_rocket_ram)
        (Audio|PlaySound2D 0 (Variables|Default|GetTimerSoundCue))))))

(event EventOnRemoval(EitherVFX-preferthisversion)
  (params (MyTarget :type Actor%20Object%20Reference) (Parameters :type Gameplay%20Cue%20Parameters%20Structure%20%28by%20ref%29) (ParticleComponents :type Array%20of%20FXSystem%20Component%20Object%20References) (AudioComponents :type Array%20of%20Audio%20Component%20Object%20References) (BurstCameraShakeInstance :type Camera%20Shake%20Base%20Object%20Reference) (BurstDecalInstance :type Decal%20Actor%20Object%20Reference))
  (bind _ticksoundtimerhandle (Variables|Default|GetTickSoundTimerHandle))
  (|CallParentFunction MyTarget Parameters ParticleComponents AudioComponents BurstCameraShakeInstance BurstDecalInstance)
  (if (Utilities|Time|IsValidTimerHandle _ticksoundtimerhandle)
    (Utilities|Time|ClearandInvalidateTimerbyHandle 0 _ticksoundtimerhandle)
    (else
      _))
  (bind _asfort_player_pawn_athena (Utilities|Casting|CastToFortPlayerPawnAthena MyTarget))
  (Utilities|IsValid _asfort_player_pawn_athena
    (:Is%20Valid
      (bind _asb_weapon_rocket_ram (Utilities|Casting|CastToB_Weapon_RocketRam (Weapon|GetCurrentWeapon _asfort_player_pawn_athena)))
      (if (FortWeapon|IsEquipped _asb_weapon_rocket_ram)
        (Audio|SpawnSoundAttached (Variables|Default|GetTimerStopSoundCue) (Variables|/Script/Engine.Actor|GetRootComponent (Weapon|GetCurrentWeapon _asfort_player_pawn_athena)))))))

(fn ConstructionScript)
