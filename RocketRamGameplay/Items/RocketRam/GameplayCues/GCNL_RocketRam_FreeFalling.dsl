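; Looping gameplay cue for the Rocket Ram free-falling state that drives an in-air visual sound indicator.
; On loop start it caches the target and starts a targeted looping indicator using the wind-down sound cue; removal stops it.
; The indicator covers friendly, neutral and hostile affiliations, but is hidden from the player controlling the current target.
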(event EventOnLoopingStart(EitherVFX-preferthisversion)
  (params (MyTarget :type Actor%20Object%20Reference) (Parameters :type Gameplay%20Cue%20Parameters%20Structure%20%28by%20ref%29) (ParticleComponents :type Array%20of%20FXSystem%20Component%20Object%20References) (AudioComponents :type Array%20of%20Audio%20Component%20Object%20References))
  (|CallParentFunction MyTarget Parameters ParticleComponents AudioComponents)
  (bind _output_get (Variables|Default|SetCurrentTarget MyTarget))
  (Utilities|IsValid _output_get
    (:Is%20Valid
      (SoundIndicator|StartTargetedSoundIndicatorLooping (Variables|Default|GetFortSoundIndicator_InAir) _output_get (Variables|Default|GetWindDownSoundCue) 1.0 0))))

(event EventOnRemoval(EitherVFX-preferthisversion)
  (params (MyTarget :type Actor%20Object%20Reference) (Parameters :type Gameplay%20Cue%20Parameters%20Structure%20%28by%20ref%29) (ParticleComponents :type Array%20of%20FXSystem%20Component%20Object%20References) (AudioComponents :type Array%20of%20Audio%20Component%20Object%20References) (BurstCameraShakeInstance :type Camera%20Shake%20Base%20Object%20Reference) (BurstDecalInstance :type Decal%20Actor%20Object%20Reference))
  (|CallParentFunction MyTarget Parameters ParticleComponents AudioComponents BurstCameraShakeInstance BurstDecalInstance)
  (SoundIndicator|StopSoundIndicatorLooping (Variables|Default|GetFortSoundIndicator_InAir)))

(fn GetAffiliationsToShowFor
  (return (Utilities|Array|MakeArray "Friendly" "Neutral" "Hostile")))

(fn ShouldShowSoundIndicator
  (params (PlayerController :type Fort%20Player%20Controller%20Object%20Reference))
  (bind _currenttarget (Variables|Default|GetCurrentTarget))
  (Utilities|IsValid _currenttarget
    (:Is%20Valid
      (return (!= PlayerController (Fortnite|GetControllerfromActor _currenttarget))))
    (:Is%20Not%20Valid
      (return true))))

(fn ConstructionScript)
