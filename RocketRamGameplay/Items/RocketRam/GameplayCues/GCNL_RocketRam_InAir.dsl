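; Looping gameplay cue that drives the Rocket Ram's in-air particle effect on a Fort player pawn.
; On loop start it caches the first FX component, casts the pawn's current weapon to B_Weapon_RocketRam and binds to its ram state event.
; Each state change sets the float uEndLifeActiveThrust to 1 for one specific state and 0 otherwise.
; On removal it unbinds from the weapon if still valid.
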(event EventOnLoopingStart(EitherVFX-preferthisversion)
  (params (MyTarget :type Actor%20Object%20Reference) (Parameters :type Gameplay%20Cue%20Parameters%20Structure%20%28by%20ref%29) (ParticleComponents :type Array%20of%20FXSystem%20Component%20Object%20References) (AudioComponents :type Array%20of%20Audio%20Component%20Object%20References))
  (Variables|Default|SetRocketRam%20FX (Utilities|Array|Get(acopy) ParticleComponents))
  (bind _asfort_player_pawn (Utilities|Casting|CastToFortPlayerPawn MyTarget))
  (bind _output_get (Variables|Default|SetPlayer _asfort_player_pawn))
  (bind _asb_weapon_rocket_ram (Utilities|Casting|CastToB_Weapon_RocketRam (Weapon|GetCurrentWeapon _output_get)))
  (bind _output_get_2 (Variables|Default|SetWeapon_RocketRam _asb_weapon_rocket_ram))
  (Default|AssignOnRamStateChangedEvent _output_get_2 (delegate Custom|OnRamStateChangedEvent)))

(event EventOnRemoval(EitherVFX-preferthisversion)
  (params (MyTarget :type Actor%20Object%20Reference) (Parameters :type Gameplay%20Cue%20Parameters%20Structure%20%28by%20ref%29) (ParticleComponents :type Array%20of%20FXSystem%20Component%20Object%20References) (AudioComponents :type Array%20of%20Audio%20Component%20Object%20References) (BurstCameraShakeInstance :type Camera%20Shake%20Base%20Object%20Reference) (BurstDecalInstance :type Decal%20Actor%20Object%20Reference))
  (bind _weapon_rocketram (Variables|Default|GetWeapon_RocketRam))
  (Utilities|IsValid _weapon_rocketram
    (:Is%20Valid
      (Default|UnbindEventfromOnRamStateChangedEvent _weapon_rocketram (delegate Custom|OnRamStateChangedEvent)))))

(event Custom|OnRamStateChangedEvent
  (params (RamState :type Enum_RocketRamState%20Enum :default "NewEnumerator0"))
  (Effects|Components|ParticleSystem|SetFloatParameter (Variables|Default|GetRocketRam%20FX) "uEndLifeActiveThrust" (Math|Float|SelectFloat 1.0 0.0 (Utilities|Enum|Equal(Enum) RamState "NewEnumerator1"))))

(fn ConstructionScript)
