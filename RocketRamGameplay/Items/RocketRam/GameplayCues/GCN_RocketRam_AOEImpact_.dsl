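; Burst gameplay cue notify for the Rocket Ram item's area-of-effect impact, handling its particles and sound indicator.
; OnBurst moves each spawned Niagara component to the cue origin and sets its global FX scale.
; That scale travels in the hit result's Time field, repurposed to carry the effect radius.
; PlaySoundIndicator shows a local gunshot-type indicator at the target, using the first burst sound and its max audible distance.
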(fn OnBurst
  (params (Target :type Actor%20Object%20Reference) (Parameters :type Gameplay%20Cue%20Parameters%20Structure%20%28by%20ref%29) (SpawnResults :type Gameplay%20Cue%20Notify%20Spawn%20Result%20Structure%20%28by%20ref%29))
  (bind (_bblockinghit _binitialoverlap _time) (Collision|BreakHitResult (Ability|GameplayCue|GetHitResult Parameters)))
  (CallFunction|PlaySoundIndicator Target)
  ;[0.515 0 1] Set the radius of the particle effect (it has the param name FX_Scale_Global).
  ; based on the value stored in the context HitResult::Time.
  (for _array_element (Utilities|Struct|BreakGameplayCueNotifySpawnResult SpawnResults)
    (bind _asniagara_particle_system_component (Utilities|Casting|CastToNiagaraComponent _array_element))
    (Transformation|SetWorldLocation _asniagara_particle_system_component (Ability|GameplayCue|GetOrigin Parameters) false true)
    (Niagara|SetNiagaraVariable(Float) _asniagara_particle_system_component "FX_Scale_Global" _time))
  (return))

(fn PlaySoundIndicator
  (params (OwningTarget :type Actor%20Object%20Reference))
  (bind _bursteffects (Variables|GCNEffects|GetBurstEffects))
  (bind (_burstparticles _burstsounds) (Utilities|Struct|BreakGameplayCueNotifyBurstEffects _bursteffects))
  (bind (_spawnconditionoverride _placementinfooverride _sound) (Utilities|Struct|BreakGameplayCueNotifySoundInfo (Utilities|Array|Get(aref) _burstsounds)))
  (Utilities|IsValid OwningTarget
    (:Is%20Valid
      (if (> (Utilities|Array|Length _burstsounds) 0)
        (Fortnite|ShowSoundIndicatorLocal :Location (Transformation|GetActorLocation OwningTarget) :MaxAudibleDistance (Variables|/Script/Engine.SoundBase|GetMaxDistance _sound) :IndicatorType "Gunshot" :Tint "(R=0.000000,G=0.000000,B=0.000000,A=0.000000)" :SoundBase _sound :AttenuationMaxRangeMultiplier 1.0 :AttenuationIntensityMultiplier 1.0)))))
